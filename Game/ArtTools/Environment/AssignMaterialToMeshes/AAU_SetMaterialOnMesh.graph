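# Editor utility Blueprint for environment art that assigns one material to a chosen slot on many selected Static Mesh assets at once.
# Its AssignMaterialToMesh function takes a slot index and a material, then loops over the selected assets.
# Each asset that casts to StaticMesh gets SetMaterial called on the asset itself, while other asset types are skipped.
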EditorUtilityBlueprint /Game/ArtTools/Environment/AssignMaterialToMeshes/AAU_SetMaterialOnMesh
graphs [
 AssignMaterialToMesh {nodes [
   n1 = AssignMaterialToMesh!() -> (then, MaterialSlot:int, NewMaterial:Material Interface Object Reference) {is=title K2Node_FunctionEntry}
   > n2 = GetSelectedAssets()
   > n3 = For Each Loop(Array:Array of Object References=@n2) -> (Array Element:Object Reference)
   > n4 = Cast To StaticMesh(Object=@n3)
   > SetMaterial!(execute, self:Static Mesh Object Reference=@n4, MaterialIndex:int=@n1, NewMaterial:Material Interface Object Reference=@n1.NewMaterial) -> (then)
  ]}
 EventGraph {links [], nodes []}
]
variables []
